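# Animation Blueprint for the power-up scythe melee weapon that adds procedural floating motion to its dynamic cube bones.
# Each update, a Perlin-style random vector function, timed by scaled real time, writes an offset into cubeTrans.
# The AnimGraph adds that offset to dyn_cube_mstr in bone space at half alpha, then runs Anim Dynamics with a soft linear spring on dyn_cube for secondary motion.
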AnimBlueprint /CorruptionGameplay/Gameplay/Items/Weapons/Melee/Scythe/Assets/Meshes/Power_Up_Scythe_Melee_AnimBP
target_skeleton=&/Power_Up_Scythe_Melee_Skeleton
graphs [
 AnimGraph {nodes [
   n1 = "Transform (Modify) Bone - Bone: dyn_cube_mstr"(
    ComponentPose=Local To Component(LocalPose=Input Pose())
    Alpha=0.5
    Translation=Get cubeTrans() -> (cubeTrans:Vector)
    Rotation="0,0,0"
    Scale="1,1,1"
   ) {
    alpha=0.5
    rotationMode=BMM_Additive
    translationMode=BMM_Additive
    translationSpace=BCS_BoneSpace
    alphaBoolBlend {blendOption=Linear}
    boneToModify {boneName=dyn_cube_mstr}
   }
   n2 = Anim Dynamics - Bone: dyn_cube(ComponentPose=@n1, Alpha=1, ExternalForce="0,0,0") {
    bGravityOverrideInSimSpace=true
    bLinearSpring=true
    gravityScale=0.07
    linearSpringConstant=10
    alphaBoolBlend {blendOption=Linear}
    alphaScaleBiasClamp {scale=0.11}
    boundBone {boneName=dyn_cube}
    componentLinearAccScale=<0.11 0.13 0.18>
    physicsBodyDefinitions [{
      sphereCollisionRadius=0
      boundBone {boneName=dyn_cube}
      boxExtents=<0.25 0.25 0.25>
      constraintSetup {angularLimitsMax=<0 0 0>, angularLimitsMin=<0 0 0>, angularTarget=<0 0 0>, linearAxesMax=<0 30 30>, linearAxesMin=<0 -30 -30>}
      localJointOffset=<0 0 0>
     }]
   }
   Output Pose(Result=Component To Local(ComponentPose=@n2)) {layerGroup=None}
  ]}
 EventGraph {nodes [
   Event BlueprintUpdateAnimation()
   > n1 = Sequence()
   > n2 = Random Vector Function!(
    execute
    CharacterPartAnimInstance:Custom Character Part Anim Instance Object Reference
    SkelMeshComponent:Mesh Component Object Reference=GetOwnerSkeletalMeshComponent()
    JointName:Name=None
    GameTimeInSeconds:double=Get customSpeed() -> (customSpeed:double)
    TimeOffset:double=0
    PerlinScale:double=1
    Frequency:Vector="0.5,0.5,0.75"
    Amplitude:Vector="0.5,0.5,1"
    Offset:Vector="0,3,2"
   ) -> (then, VectorOut:Vector)
   > Set cubeTrans(cubeTrans:Vector=@n2) -> (Output_Get:Vector)
   Set customSpeed(execute=@n1.then_1, customSpeed:double=float / float(A=GetRealTimeSeconds(), B=1.25) {is=title K2Node_CallFunction}) -> (Output_Get:double)
   TryGetPawnOwner()
  ]}
]
variables [
 cubeTrans {type=Vector}
 customSpeed {type=number}
 frontEndAlpha {type=number}
]
